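# Unlit, two-sided masked material for sliding spark effects on spline meshes, usable with Niagara mesh particles.
# SplineThicken takes its width from a Dynamic Parameter, which also drives squash and upward pan of the mask UVs.
# Particle color blends toward an intensity-boosted copy along V, then passes through MF_FX_EmissiveAdjustments.
# The spark shape is procedural and texture-free, multiplying a squashed, panned sphere mask by a static one, both power-shaped.
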Material /Game/Effects/Fort_Effects/Gameplay/Sliding/M_Sparks_Spline_Mesh
graph {
 properties {
  bHasPixelAnimation=true
  bUsedWithNiagaraMeshParticles=true
  blendMode=BLEND_Masked
  opacityMaskClipValue=0.2
  pixelDepthOffsetMode=PDOM_Legacy
  shadingModel=MSM_Unlit
  twoSided=true
  usedShadingModels=Unlit
 }
 nodes [
  Constant() {r=2}
  n1 = DynamicParameter() {defaultValue=<rgba 1 0 6 1>, paramNames [Squash, Pan Upwards, Spline Thickness, Param4]}
  MaterialFunctionCall(WidthBase=@n1.Spline Thickness, WidthTip=@n1.Spline Thickness) {materialFunction=&SplineThicken}
  n2 = ParticleColor()
  n3 = ScalarParameter() {defaultValue=10, parameterName=Color Intensity}
  n4 = ComponentMask(None=TextureCoordinate()) {g=true}
  MaterialFunctionCall(Emissive=LinearInterpolate(A=@n2.RGB, B=Multiply(A=@n2.RGB, B=@n3) {constB=10}, Alpha=Power(Base=@n4))) {materialFunction=&MF_FX_EmissiveAdjustments}
  n5 = Constant() {r=1}
  n6 = DynamicParameter() {defaultValue=<rgba 1 0 6 1>, paramNames [Squash, Pan Upwards, Spline Thickness, Param4]}
  n7 = Add(A=Multiply(A=TextureCoordinate(), B=AppendVector(A=@n5, B=@n6.Squash)), B=AppendVector(A=Constant(), B=@n6.Pan Upwards))
  n8 = MaterialFunctionCall(Float2=Abs(None=ConstantBiasScale(None=@n7) {bias=-0.5, scale=2})) {materialFunction=&BreakOutFloat2Components}
  n9 = MaterialFunctionCall(X=@n8.R, Y=Power(Base=@n8.G) {constExponent=3}) {materialFunction=&MakeFloat2}
  n10 = Constant()
  n11 = Constant() {r=1}
  n12 = Constant()
  n13 = TextureCoordinate()
  n14 = MaterialFunctionCall(Float2=Abs(None=ConstantBiasScale(None=@n13) {bias=-0.5, scale=2})) {materialFunction=&BreakOutFloat2Components}
  n15 = Power(Base=@n14.G, Exp=ScalarParameter() {defaultValue=3, parameterName=SphereMask Power}) {constExponent=3}
  Multiply(
   A=Power(Base=SphereMask(A=@n9.Result, B=@n10, Radius=@n11, Hardness=@n12), Exp=Constant() {r=2}) {constExponent=3}
   B=SphereMask(
    A=MaterialFunctionCall(X=@n14.R, Y=@n15) {materialFunction=&MakeFloat2}.Result
    B=Constant()
    Radius=Constant() {r=1}
    Hardness=Constant()
   )
  )
 ]
}
