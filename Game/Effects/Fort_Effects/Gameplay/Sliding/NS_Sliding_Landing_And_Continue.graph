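# Niagara one-shot burst effect for landing into a continued slide, built from dust, rock debris and spike emitters.
# Dust sprites spawn on a ring, narrowed to an arc by a quality-level check, then get flattened, forward-boosted velocity and curve-driven material parameters.
# GPU mesh rocks scatter under gravity and drag with collision disabled.
# Very short-lived, position-oriented spike meshes stretch out from a forward arc and fade.
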NiagaraSystem /Game/Effects/Fort_Effects/Gameplay/Sliding/NS_Sliding_Landing_And_Continue
dependencies {
 usedDataInterfaces [&NiagaraDataInterfaceDebugDraw, &NiagaraDataInterfaceCurve, &NiagaraDataInterfaceColorCurve]
 usedDynamicInputs []
 usedModules [&SystemState]
}
emitters {
 Dust001 {
  emitterSpawnScript [/SetVariables(Emitter.bUseLowerSpawnCount:NiagaraBool=&SetBoolByIntComparison(A=$Engine.QualityLevel, B=3, Comparison Type=::NewEnumerator6)) {isSetParameters=true}]
  emitterUpdateScript [
   &EmitterState(Life Cycle Mode=::NewEnumerator1, Loop Behavior=::NewEnumerator1, Loop Duration=1.5, UseLoopDelay=false)
   &SpawnBurst_Instantaneous(Spawn Count=7, Use Spawn Probability=false, UseLoopCountLimit=false)
  ]
  particleSpawnScript [
   &InitializeParticle(
    Color=&MakeCustomLinearColorFromBool(Bool=&RandomBool(), True Color=<rgba 0.25 0.25 0.25 0.3>, False Color=<rgba 0.1938 0.2 0.1875 0.25>)
    Lifetime Max=1
    Lifetime Min=0.4
    Lifetime Mode=::NewEnumerator1
    Position Offset=<0 0 3.5>
    Sprite Rotation Mode=::NewEnumerator2
    Sprite Size Mode=::NewEnumerator1
    Uniform Sprite Size Max=85
    Uniform Sprite Size Min=70
    UsePositionOffset=true
    Write Material Random=true
   )
   &TorusLocation(
    Align Torus With Axis=false
    Handle Radius=1
    Large Radius=15
    MaskSpawnGroup=false
    Torus Distribution Mode:ENiagaraTorusDistributionMode=::NewEnumerator1
    U Position:NiagaraFloat=&MakeCustomFloatFromBool(
     Bool=$Emitter.bUseLowerSpawnCount
     True Float=&RemapRange(
      Input Value To Remap=&ReturnNormalizedExecIndex()
      Input Min=0
      Input Max=1
      Output Min=-0.25
      Output Max=0.25
     )
     False Float=&ReturnNormalizedExecIndex()
    )
   )
   &AddVelocity(
    Module.Use Falloff Distance=false
    Override Seed=false
    Velocity Mode=::NewEnumerator1
    Velocity Speed=&RandomRangeFloat(Minimum=200, Maximum=500)
   )
   /SetVariables(Particles.Velocity=&Multiply_Vector(A=$Particles.Velocity, B=<1 1 0.001>)) {isSetParameters=true}
   &AddVelocity(Velocity=<1 0 0>, Velocity Speed Scale=&RandomRangeFloat(Minimum=750, Maximum=925)) {module=AddVelocity001}
  ]
  particleUpdateScript [
   &Drag(Drag=4, Use Linear Drag=true, Use Rotational Drag=false)
   &AccelerationForce(Acceleration=&RandomRangeVector(Minimum=<0 0 0>, Maximum=<0 0 0>))
   &SolveForcesAndVelocity(Clamp Velocity=false, Limit Acceleration=false)
   &ParticleState()
   &DynamicMaterialParameters(
    Index 0 Param 1=&FloatFromCurve(FloatCurve=DataInterface() {propertyValues {Curve {keys [<at 0 0>,<at 1 1>]}}})
    Index 0 Param 4=1
    Index 2 Param 1=&FloatFromCurve(FloatCurve=DataInterface() {propertyValues {Curve {keys [
         {arriveTangent=3.7909, interpMode=RCIM_Cubic, leaveTangent=3.7909, tangentMode=RCTM_User, time=0, value=0}
         {arriveTangent=1.8782, interpMode=RCIM_Cubic, leaveTangent=1.8782, tangentMode=RCTM_User, time=0.2, value=0.55}
         {arriveTangent=0.0797, interpMode=RCIM_Cubic, leaveTangent=0.0797, tangentMode=RCTM_User, time=1, value=1}
        ]}}})
    Index 2 Param 2=0.4
    Index 2 Param 3=&FloatFromCurve(FloatCurve=DataInterface() {propertyValues {Curve {keys [
         {arriveTangent=2.5608, interpMode=RCIM_Cubic, leaveTangent=2.5608, tangentMode=RCTM_User, time=0, value=0}
         {arriveTangent=1.5984, interpMode=RCIM_Cubic, leaveTangent=1.5984, tangentMode=RCTM_User, time=0.2, value=0.5}
         {arriveTangent=0.5453, interpMode=RCIM_Cubic, leaveTangent=0.5453, tangentMode=RCTM_User, time=1, value=1}
        ]}}})
    Index 2 Param 4=&FloatFromCurve(FloatCurve=DataInterface() {propertyValues {Curve {keys [<at 0 0>,<at 1 2>]}}}, Scale Curve=0.25)
    Module.Param0WriteEnabled=true
    Module.Param0WriteEnabled2=true
    Module.Param1WriteEnabled=true
    Module.Param1WriteEnabled2=true
    Module.Param2WriteEnabled=true
    Module.Param2WriteEnabled2=true
    Module.Param3WriteEnabled=true
    Module.Param3WriteEnabled2=true
    Write Parameter Index 2=true
   )
   &ScaleColor(
    Scale Alpha=&FloatFromCurve(FloatCurve=DataInterface() {propertyValues {Curve {keys [
         {arriveTangent=10, interpMode=RCIM_Cubic, leaveTangent=10, time=0, value=0}
         {arriveTangent=2, interpMode=RCIM_Cubic, leaveTangent=2, time=0.1, value=1}
         {arriveTangent=-1.1111, interpMode=RCIM_Cubic, leaveTangent=-1.1111, time=0.4, value=0.8}
         {arriveTangent=-1.3333, interpMode=RCIM_Cubic, leaveTangent=-1.3333, time=1, value=0}
        ]}}})
    ScaleA=true
    ScaleRGB=true
   )
   &ScaleSpriteSize(Uniform Curve Sprite Scale=DataInterface() {propertyValues {Curve {keys [
        {arriveTangent=1.0717, interpMode=RCIM_Cubic, leaveTangent=1.0717, tangentMode=RCTM_User, time=0, value=1}
        {arriveTangent=0.0885, interpMode=RCIM_Cubic, leaveTangent=0.0885, tangentMode=RCTM_User, time=1, value=1.7}
       ]}}})
   &CameraOffset(Camera Offset Amount=10)
  ]
  rendererClasses [&NiagaraSpriteRendererProperties]
  renderers [
   {rendererClass=&NiagaraSpriteRendererProperties}
   {rendererClass=&NiagaraSpriteRendererProperties, rendererIndex=1}
  ]
 }
 Rock {
  simTarget=GPUComputeSim
  emitterUpdateScript [
   &EmitterState(Inactive Response=::NewEnumerator1, Life Cycle Mode=::NewEnumerator1, Loop Behavior=::NewEnumerator1, Loop Duration=1, UseLoopDelay=false)
   &SpawnBurst_Instantaneous(Spawn Count=6, Use Spawn Probability=false, UseLoopCountLimit=false)
  ]
  particleSpawnScript [
   &InitializeParticle(
    Color=&MakeCustomLinearColorFromBool(Bool=&RandomBool(), True Color=<rgba 0.2917 0.2234 0.1598 1>, False Color=<rgba 0.0677 0.0429 0.0342 1>)
    Lifetime Max=0.5
    Lifetime Min=0.1
    Lifetime Mode=::NewEnumerator1
    Mesh Scale Mode=::NewEnumerator1
    Mesh Uniform Scale Max=0.014
    Mesh Uniform Scale Min=0.01
    Position Offset=<45.702 0 11.3574>
    Position Offset Coordinate Space=::Local
    UsePositionOffset=true
    Write Material Random=true
   )
   &InitialMeshOrientation(Use Rotation Vector=false)
   &TorusLocation(Align Torus With Axis=false, Handle Radius=5, Large Radius=65.4979, MaskSpawnGroup=false, Module.Override Randomness=false)
   &AddVelocity(
    Module.Use Falloff Distance=false
    Origin Offset=<0 0 -2.0051>
    Override Seed=false
    Velocity Mode=::NewEnumerator1
    Velocity Speed=&RandomRangeFloat(Minimum=100, Maximum=800)
   )
  ]
  particleUpdateScript [
   &Drag(Drag=2, Use Linear Drag=true, Use Rotational Drag=false)
   &AccelerationForce(Acceleration=<0 0 -800>)
   &Collision(
    Age Colliding Particles=false
    GPU Collision Type=::NewEnumerator2
    Module.Control Rest State Via Surface Normal=false
    Module.EnableMaxCollisionCount=false
    Module.Randomize Collision Normal Vector=false
    Particle Radius=0.3
    Particle Radius Scale=0.2
    Radius Calculation Type=::NewEnumerator3
    Restitution=0.2
   ) {enabled=false, module=Collision001}
   &SolveForcesAndVelocity(Clamp Velocity=false, Limit Acceleration=false)
   &ParticleState()
   &ScaleMeshSize(Scale Factor=&Multiply_VectorByFloat(
     Vector=<1 1 1>
     Float=&FloatFromCurve(
      FloatCurve=DataInterface() {propertyValues {Curve {keys [<at 0.6 1>,<at 1 0>]}}}
      Scale Curve=&RandomRangeFloat(Minimum=0.8, Maximum=1.1)
     )
    ))
   &UpdateMeshOrientation(Rotation Rate=&RandomRangeFloat(Minimum=-1, Maximum=1), Rotation Vector=<1 1 1>)
  ]
  rendererClasses [&NiagaraMeshRendererProperties]
  renderers [{rendererClass=&NiagaraMeshRendererProperties, rendererIndex=0}]
 }
 SpikeOut {
  emitterUpdateScript [
   &EmitterState()
   &SpawnBurst_Instantaneous(Spawn Count=3, Use Spawn Probability=false, UseLoopCountLimit=false)
  ]
  particleSpawnScript [
   &InitializeParticle(
    Color=<rgba 0.7969 0.6536 0.5973 0.1>
    Lifetime Max=0.08
    Lifetime Min=0.04
    Lifetime Mode=::NewEnumerator1
    Mesh Scale Max=<2 1.2 1>
    Mesh Scale Min=<0.5 0.6 1>
    Mesh Scale Mode=::NewEnumerator2
    UsePositionOffset=false
    Write Material Random=true
   )
   &CameraOffset(Camera Offset Amount=10)
   &TorusLocation(
    Align Torus With Axis=false
    Handle Radius=1
    Large Radius=20
    MaskSpawnGroup=false
    Torus Distribution Mode:ENiagaraTorusDistributionMode=::NewEnumerator1
    U Position:NiagaraFloat=&RandomRangeFloat(Minimum=0.25, Maximum=-0.25)
   )
   &InitialMeshOrientation(
    Mesh Axis To Orient=<1 0 -0.753>
    Mesh Orientation Mode=::NewEnumerator2
    Orientation Coordinate Space=::Local
    Orientation Vector=&NormalizeVector(
     Vector To Normalize=Linked() {linkedVariable {name=Particles.Position, type=NiagaraPosition}}
     Safe Normalize=true
     Safe Fallback Threshold=0
     Safe Fallback Vector=<0 0 0.001>
    )
    Rotation=&RandomRangeVector(Minimum=<0 0.5 0>, Maximum=<0 0.6 0>)
    Use Rotation Vector=true
   )
   &SystemLocation(MaskSpawnGroup=false, Offset=<11.2607 0 0>) {enabled=false}
  ]
  particleUpdateScript [
   &ParticleState()
   &ScaleColor(
    Scale Alpha=&FloatFromCurve(FloatCurve=DataInterface() {propertyValues {Curve {keys [
         {arriveTangent=-0.7355, interpMode=RCIM_Cubic, leaveTangent=-0.7355, tangentMode=RCTM_User, time=0.5, value=1}
         {arriveTangent=-3.1363, interpMode=RCIM_Cubic, leaveTangent=-3.1363, tangentMode=RCTM_User, time=1, value=0}
        ]}}})
    ScaleA=true
    ScaleRGB=true
   )
   &ScaleMeshSize(Scale Factor=&MakeVector(
     X=&FloatFromCurve(
      FloatCurve=DataInterface() {propertyValues {Curve {keys [{arriveTangent=2.7289, interpMode=RCIM_Cubic, leaveTangent=2.7289, tangentMode=RCTM_User, time=0, value=0}, <at 1 1>]}}}
      Scale Curve=1.75
     )
     Y=&FloatFromCurve(FloatCurve=DataInterface() {propertyValues {Curve {keys [{arriveTangent=2.7289, interpMode=RCIM_Cubic, leaveTangent=2.7289, tangentMode=RCTM_User, time=0, value=0}, <at 1 1>]}}})
     Z=1
    ))
  ]
  rendererClasses [&NiagaraMeshRendererProperties]
  renderers [{rendererClass=&NiagaraMeshRendererProperties, rendererIndex=0}]
 }
}
summary {
 systemName=NS_Sliding_Landing_And_Continue
 emitters [
  {emitterName=Rock, simTarget=GPUComputeSim, rendererClasses [&NiagaraMeshRendererProperties]}
  {emitterName=Dust001, rendererClasses [&NiagaraSpriteRendererProperties]}
  {emitterName=SpikeOut, rendererClasses [&NiagaraMeshRendererProperties]}
 ]
 userVariables [
  User.bIsFrontEndPreview {type=NiagaraBool}
  User.bIsFullyDeployed {type=NiagaraBool}
  User.ForwardDot {type=NiagaraFloat, defaultValue=0, description=""}
  User.IsSpawnDust {type=NiagaraBool, defaultValue=true}
 ]
}
