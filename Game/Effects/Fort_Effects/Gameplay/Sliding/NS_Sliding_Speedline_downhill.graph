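# Niagara system for the speedline effect shown during downhill sliding.
# The active Mesh emitter uses a burst plus a slow spawn rate to emit very faint, oriented meshes whose alpha fades in and out over life.
# A disabled GPU Speedlines emitter holds torus-spawned sprites pushed from a point and stretched by speed.
# User parameters for deployment and forward/right dot are declared, but no shown module reads them.
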NiagaraSystem /Game/Effects/Fort_Effects/Gameplay/Sliding/NS_Sliding_Speedline_downhill
dependencies {
 usedDataInterfaces [&NiagaraDataInterfaceCurve, &NiagaraDataInterfaceVector2DCurve]
 usedDynamicInputs []
 usedModules [&SystemState]
}
emitters {
 Mesh {
  emitterUpdateScript [
   &EmitterState(Life Cycle Mode=::NewEnumerator1, Loop Duration=1, UseLoopDelay=false)
   &SpawnBurst_Instantaneous(Use Spawn Probability=false, UseLoopCountLimit=false)
   &SpawnRate(SpawnRate=2, Use Spawn Probability=false)
  ]
  particleSpawnScript [
   &InitializeParticle(Color=<rgba 1 1 1 0.01>, UsePositionOffset=false, Write Material Random=true)
   &InitialMeshOrientation(Mesh Orientation Mode=::NewEnumerator7, Rotation=<0 0.5 0>, Use Rotation Vector=true)
  ]
  particleUpdateScript [
   &ParticleState()
   &ScaleColor(
    Scale Alpha=&FloatFromCurve(FloatCurve=DataInterface() {propertyValues {Curve {keys [
         {arriveTangent=5.8909, interpMode=RCIM_Cubic, leaveTangent=5.8909, tangentMode=RCTM_User, time=0, value=0}
         <at 0.3 1 RCIM_Cubic>
         {arriveTangent=-1.4286, interpMode=RCIM_Cubic, leaveTangent=-1.4286, time=1, value=0}
        ]}}})
    ScaleA=true
    ScaleRGB=true
   )
  ]
  rendererClasses [&NiagaraMeshRendererProperties]
  renderers [{rendererClass=&NiagaraMeshRendererProperties, rendererIndex=0}]
 }
 Speedlines {
  bEnabled=false
  simTarget=GPUComputeSim
  emitterUpdateScript [
   &EmitterState(Delay First Loop Only=true, Inactive Response=::NewEnumerator1, Life Cycle Mode=::NewEnumerator1, Loop Duration=1, UseLoopDelay=true)
   &SpawnRate(SpawnRate=800, Use Spawn Probability=true)
   &SpawnPerUnit(
    Max Movement Threshold=1000
    Module.Use Max Movement Threshold=false
    Module.Use Movement Tolerance=true
    Movement Tolerance=30
    Spawn Probability=1
    Spawn Spacing=1
    Use Spawn Probability=true
   ) {enabled=false}
  ]
  particleSpawnScript [
   /Niagara/Modules/Spawn/Initialization/InitializeParticle(
    Color=&MakeLinearColorFromVectorAndFloat("Vector (RGB)"=<1 1 1>, "Float (Alpha)"=0.15)
    Lifetime=&UniformRangedFloat(Minimum=0.1, Maximum=0.15)
    Module.Write Color=true
    Module.Write Lifetime=true
    Module.Write Mass=true
    Module.Write Material Random=true
    Module.Write Position=false
    Module.Write Scale=false
    Module.Write SpriteRotation=false
    Module.Write SpriteSize=true
    Position=&SimulationPosition()
    Sprite Size=&UniformRangedVector2D(Minimum=<1 6>, Maximum=<2 9>)
   )
   &CylinderLocation(
    Cylinder Height=60
    Cylinder Radius=&UniformRangedFloat(Minimum=30, Maximum=240)
    MaskSpawnGroup=false
    Override Seed=false
    UseOwnerScale=true
   ) {enabled=false}
   &TorusLocation(Align Torus With Axis=true, Handle Radius=10, Large Radius=36, MaskSpawnGroup=false, Module.Override Randomness=false, Torus Axis=<1 0 0>)
   &CylinderLocation(
    Cylinder Height=120
    Cylinder Radius=&RandomRangeFloat(Minimum=300, Maximum=240)
    MaskSpawnGroup=false
    Override Seed=false
    UseOwnerScale=true
   ) {enabled=false, module=CylinderLocation001}
   &AddVelocityFromPoint(Module.Use Falloff Distance=false, Velocity Strength=&UniformRangedFloat(Minimum=1000, Maximum=300))
   &AddVelocity(Velocity=&RandomRangeVector(Minimum=<0 0 0>, Maximum=<0 0 0>))
  ]
  particleUpdateScript [
   &ParticleState()
   &SolveForcesAndVelocity(Clamp Velocity=false, Limit Acceleration=false)
   &ScaleSpriteSize(Scale Factor=&Vector2DFromCurve(
     Vector2Curve=DataInterface() {propertyValues {XCurve {keys [<at 0 0.5 RCIM_Cubic>,<at 0.5 1 RCIM_Cubic>]}, YCurve {keys [<at 0 0 RCIM_Cubic>,<at 0.5 1 RCIM_Cubic>]}}}
     ScaleCurve=&Vector2DFromFloat(Value=1)
    ))
   &ScaleSpriteSizeBySpeed(Max Scale Factor=<1 50>, Min Scale Factor=<0.01 0.05>, Velocity Threshold=1)
   &ScaleColor(
    Scale Alpha=&FloatFromCurve(FloatCurve=DataInterface() {propertyValues {Curve {keys [
         {arriveTangent=5.2641, interpMode=RCIM_Cubic, leaveTangent=5.2641, tangentMode=RCTM_User, time=0, value=0}
         {arriveTangent=0.7132, interpMode=RCIM_Cubic, leaveTangent=0.7132, tangentMode=RCTM_User, time=0.3, value=1.0207}
         {arriveTangent=-2.9069, interpMode=RCIM_Cubic, leaveTangent=-2.9069, tangentMode=RCTM_User, time=1, value=0}
        ]}}})
    Scale RGB=&VectorFromFloat(Value=1)
    ScaleA=true
    ScaleRGB=true
   )
  ]
  rendererClasses [&NiagaraSpriteRendererProperties]
  renderers [{rendererClass=&NiagaraSpriteRendererProperties}]
 }
}
summary {
 systemName=NS_Sliding_Speedline_downhill
 emitters [
  {bEnabled=false, emitterName=Speedlines, simTarget=GPUComputeSim, rendererClasses [&NiagaraSpriteRendererProperties]}
  {emitterName=Mesh, rendererClasses [&NiagaraMeshRendererProperties]}
 ]
 userVariables [
  User.bisFullyDeployed {type=NiagaraBool}
  User.ForwardDot {type=NiagaraFloat, defaultValue=0, description=""}
  User.RightDot {type=NiagaraFloat, defaultValue=0, description=""}
 ]
}
